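# Material function that converts a world position into UVW coordinates within a custom bounding box given by Min and Size.
# It adds the actor position's Z component, as a scalar, to Min, subtracts that from World Position and divides by Size.
# A static Clamp switch optionally clamps the result to 0-1 before it outputs as UVW and separate U, V, W channels.
# A TransformPosition node on World Position feeds no output.
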MaterialFunction /Engine/ArtTools/RenderToTexture/MaterialFunctions/CustomBoundingBox_0-1_UVW
graph {nodes [
  n1 = FunctionInput(Preview=WorldPosition() {worldPositionShaderOffset=WPT_ExcludeAllShaderOffsets}.XYZ) {bUsePreviewValueAsDefault=true, inputName=World Position}
  n2 = FunctionInput() {bUsePreviewValueAsDefault=true, inputName=Min, sortPriority=1}
  n3 = ComponentMask(None=ActorPositionWS()) {b=true}
  n4 = Divide(A=Subtract(A=@n1, B=Add(A=@n2, B=@n3)), B=FunctionInput() {bUsePreviewValueAsDefault=true, inputName=Size, sortPriority=2})
  n5 = FunctionInput(Preview=StaticBool()) {bUsePreviewValueAsDefault=true, inputName=Clamp, inputType=FunctionInput_StaticBool, sortPriority=3, previewValue=<1 0 0 1>}
  n6 = FunctionOutput(None=StaticSwitch(True=Clamp(None=@n4), False=@n4, Value=@n5)) {outputName=UVW}
  FunctionOutput(None=ComponentMask(None=@n6) {r=true}) {outputName=U, sortPriority=1}
  FunctionOutput(None=ComponentMask(None=@n6) {g=true}) {outputName=V, sortPriority=2}
  FunctionOutput(None=ComponentMask(None=@n6) {b=true}) {outputName=W, sortPriority=3}
  TransformPosition(None=@n1) {transformSourceType=TRANSFORMPOSSOURCE_World}
 ]}
