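# Material function in the RenderToTexture library that prepares normals for baking imposter sprites only, simplifying the local transform setup.
# An optional Normal input, defaulting to a flat +Z normal, passes through a vector Transform node left at its default spaces.
# The transformed vector is multiplied by a constant that flips X and Y and keeps Z, then output as the baked normal.
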MaterialFunction /Engine/ArtTools/RenderToTexture/MaterialFunctions/TransformImposterNormals
graph {
 properties {
  bExposeToLibrary=true
  description=For Baking Imposter Sprites Only. This makes it a bit simpler to set up the local transform settings for baking imposters.
  libraryCategoriesText [Misc, RenderToTexture]
 }
 nodes [
  n1 = FunctionInput() {
   bUsePreviewValueAsDefault=true
   description="Input Normal Map. If none, 0,0,1 will be used by default."
   inputName=Normal
   previewValue=<0 0 1 1>
  }
  FunctionOutput(None=Multiply(A=Transform(None=@n1), B=Constant3Vector() {constant=<rgba -1 -1 1 0>})) {description="For Baking Imposter Sprites Only!"}
 ]
}
