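# Material function that maps world position to UVW coordinates within the object's scaled bounding sphere, with optional light-vector projection, for render-to-texture tools.
# Position becomes local, or object-relative when blueprint depth capture supplies offsets.
# An optional inverse basis transform applies the light vector; the cheaper default projects from above.
# Subtracting the sphere minimum and dividing by diameter normalizes the result, optionally clamped, output as UVW and separate channels.
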MaterialFunction /Engine/ArtTools/RenderToTexture/MaterialFunctions/BoundingSphere_LightTransform
graph {nodes [
  n1 = FunctionInput(Preview=WorldPosition() {worldPositionShaderOffset=WPT_ExcludeAllShaderOffsets}.XYZ) {bUsePreviewValueAsDefault=true, inputName=World Position}
  n2 = Subtract(A=@n1, B=ObjectPositionWS()) {constA=0}
  Add(A=@n2, B=Subtract(A=ActorPositionWS(), B=ObjectPositionWS()) {constA=0})
  n3 = FunctionInput() {bUsePreviewValueAsDefault=true, inputName=Scale Z, inputType=FunctionInput_Scalar, sortPriority=2, previewValue=<1 0 0 1>}
  AppendVector(A=Constant2Vector() {g=1, r=1}, B=@n3)
  n4 = FunctionInput(Preview=StaticBool() {value=true}) {
   bUsePreviewValueAsDefault=true
   description="When true, transforms the coordinate system from world to local space. Defaults to true, should only be disabled during blueprint capture of depth when offsets are provided by the blueprint."
   inputName=Transform to Local space
   inputType=FunctionInput_StaticBool
   sortPriority=5
   previewValue=<1 0 0 1>
  }
  n5 = StaticSwitch(True=TransformPosition(None=@n1) {transformSourceType=TRANSFORMPOSSOURCE_World}, False=@n2, Value=@n4)
  n6 = FunctionInput() {
   bUsePreviewValueAsDefault=true
   description="Only used if Transform Position is enabled. Defaults to 0,0,1 or from noon which is the cheapest. Only needed if projected is from the side. If you need a 90 degree rotation, use swizzle on world position rather than transform using a cardinal axis vector."
   inputName=Light Vector
   sortPriority=6
   previewValue=<0.01 0 1 1>
  }
  n7 = MaterialFunctionCall!(Vector1=@n6, Vector2=Constant3Vector() {constant=<rgba 1 0 0 0>}) {materialFunction=/Engine/Functions/Engine_MaterialFunctions02/Math/CreateThirdOrthogonalVector}
  n8 = MaterialFunctionCall!(VectorToTransform=@n5, BasisX=@n7.Vector2, BasisY=@n7.Vector2, BasisZ=@n7.Vector2) {materialFunction=/Engine/Functions/Engine_MaterialFunctions02/Math/InverseTransformMatrix}
  n9 = FunctionInput(Preview=StaticBool()) {
   bUsePreviewValueAsDefault=true
   description="When false (default), the projection will be from above (0,0,1) because that is cheaper than transforming into a custom vector. If necessary you can enable transform and supply a custom vector. If your custom vector is a cardinal axis like 1,0,0 or 0,1,0 consider swizzling worldposition which is cheaper than a transform."
   inputName=Custom Light Vector
   inputType=FunctionInput_StaticBool
   sortPriority=5
   previewValue=<1 0 0 1>
  }
  n10 = FunctionInput() {bUsePreviewValueAsDefault=true, inputName=Scale XY, inputType=FunctionInput_Vector2, sortPriority=1, previewValue=<1 1 0 1>}
  n11 = Multiply(A=ObjectRadius(), B=AppendVector(A=@n10, B=@n3))
  n12 = MaterialFunctionCall() {materialFunction=/Engine/Functions/Engine_MaterialFunctions02/WorldPositionOffset/ObjectScale}
  n13 = Divide(A=@n11, B=Multiply(A=@n12.Scale XYZ) {constB=-1}) {constB=-1}
  n14 = Multiply(A=@n11) {constB=2, desc=Object Diameter}
  n15 = Multiply(A=@n14) {constB=-0.5}
  n16 = Divide(
   A=Subtract(A=StaticSwitch(True=@n8.Transformed Vector 3, False=@n5, Value=@n9), B=StaticSwitch(True=@n13, False=@n15, Value=@n4)) {desc=Min}
   B=StaticSwitch(True=Divide(A=@n14, B=@n12.Scale XYZ) {constB=1}, False=@n14, Value=@n4)
  ) {desc=Size}
  n17 = FunctionInput(Preview=StaticBool()) {bUsePreviewValueAsDefault=true, inputName=Clamp, inputType=FunctionInput_StaticBool, sortPriority=4, previewValue=<1 0 0 1>}
  n18 = FunctionOutput(None=StaticSwitch(True=Clamp(None=@n16), False=@n16, Value=@n17)) {outputName=UVW}
  FunctionOutput(None=ComponentMask(None=@n18) {b=true}) {outputName=W, sortPriority=3}
  FunctionOutput(None=ComponentMask(None=@n18) {r=true}) {outputName=U, sortPriority=1}
  FunctionOutput(None=ComponentMask(None=@n18) {g=true}) {outputName=V, sortPriority=2}
 ]}
