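# Material function that computes a shadow mask and an internal transmission mask from a depth-style shadow volume texture.
# BoundingSphere_LightTransform projects world position along the projection vector into UV and depth, optionally replaced by custom UVs.
# Texture depth minus local depth plus self-shadow offset is scaled by accumulation distance for shadow, or opacity for transmission.
# An optional fake blur averages offset RGB depth comparisons and squares the result.
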MaterialFunction /Engine/ArtTools/RenderToTexture/MaterialFunctions/VolumeShadowMap
graph {nodes [
  n1 = FunctionInput() {bUsePreviewValueAsDefault=true, inputName=Opacity, inputType=FunctionInput_Scalar, sortPriority=9}
  n2 = Multiply(A=@n1) {constB=50}
  n3 = FunctionInput() {bUsePreviewValueAsDefault=true, inputName=Range, inputType=FunctionInput_Scalar, sortPriority=1, previewValue=<1 0 0 1>}
  n4 = FunctionInput() {bUsePreviewValueAsDefault=true, inputName=Projection Vector, sortPriority=3, previewValue=<0 0 1 1>}
  n5 = MaterialFunctionCall!(
   World Position=FunctionInput(Preview=WorldPosition() {worldPositionShaderOffset=WPT_ExcludeAllShaderOffsets}.XYZ) {bUsePreviewValueAsDefault=true, inputName=WorldPosition}
   Scale XY
   Scale Z=@n3
   Clamp
   Transform to Local space
   Custom Light Vector=FunctionInput(Preview=StaticBool()) {bUsePreviewValueAsDefault=true, inputName=Transform Position, inputType=FunctionInput_StaticBool, sortPriority=2}
   Light Vector=@n4
  ) {materialFunction=&/BoundingSphere_LightTransform}
  n6 = ComponentMask(None=@n5.UVW) {g=true, r=true}
  n7 = FunctionInput(Preview=StaticBool()) {bUsePreviewValueAsDefault=true, inputName=Use Custom UVs, inputType=FunctionInput_StaticBool, sortPriority=4}
  n8 = StaticSwitch(
   True=FunctionInput(Preview=TextureCoordinate() {coordinateIndex=1}) {bUsePreviewValueAsDefault=true, inputName="Custom UV,  XY", inputType=FunctionInput_Vector2, sortPriority=5}
   False=@n6
   Value=@n7
  )
  n9 = FunctionInput(Preview=TextureObject() {texture=/Engine/ArtTools/RenderToTexture/Textures/T_EV_BlankWhite_01}) {bUsePreviewValueAsDefault=true, inputName=Shadow Volume Texture, inputType=FunctionInput_Texture2D, sortPriority=7}
  n10 = TextureCoordinate() {coordinateIndex=1}
  n11 = StaticSwitch(
   True=FunctionInput(Preview=ComponentMask(None=@n10) {r=true}) {bUsePreviewValueAsDefault=true, inputName="Custom UV,  Z", inputType=FunctionInput_Scalar, sortPriority=6}
   False=@n5.W
   Value=@n7
  ) {desc=Local Z}
  n12 = FunctionInput() {bUsePreviewValueAsDefault=true, inputName=Self Shadow Offset, inputType=FunctionInput_Scalar, sortPriority=8}
  n13 = Add(
   A=Subtract(A=TextureSample(UVs=@n8, Tex=@n9) {desc=Volume Map, texture=&T_EV_BlankWhite_01}.R, B=@n11)
   B=@n12
  )
  n14 = OneMinus(None=Clamp(None=Multiply(A=@n2, B=@n13)))
  n15 = Multiply(A=@n14, B=@n14)
  n16 = FunctionInput() {bUsePreviewValueAsDefault=true, inputName=Shadow Accumulation Distance, inputType=FunctionInput_Scalar, sortPriority=9}
  n17 = Clamp(None=@n16)
  n18 = TextureSample(UVs=@n8, Tex=@n9) {desc=Volume Map, texture=&T_EV_BlankWhite_01}
  n19 = Subtract(A=@n18.R, B=@n11)
  n20 = Clamp(None=Multiply(A=Multiply(A=@n17) {constB=128}, B=Add(A=@n19, B=@n12)))
  n21 = Add(A=Subtract(A=@n18.G, B=@n11), B=@n12)
  n22 = Clamp(None=Multiply(A=Multiply(A=@n16) {constB=128}, B=@n21))
  n23 = Add(A=Subtract(A=@n18.B, B=@n11), B=@n12)
  n24 = Multiply(A=Multiply(A=@n16) {constB=128}, B=@n23)
  n25 = OneMinus(None=LinearInterpolate(A=LinearInterpolate(A=@n20, B=@n22), B=Clamp(None=@n24)) {constAlpha=0.33})
  Add(A=@n15, B=@n25)
  Clamp(None=@n1)
  Clamp(None=@n21)
  Clamp(None=@n23)
  FunctionInput() {bUsePreviewValueAsDefault=true, inputName=Transmission Mip Bias, inputType=FunctionInput_Scalar, sortPriority=9}
  n26 = Multiply(A=@n25, B=@n25)
  n27 = OneMinus(None=@n20)
  n28 = FunctionInput(Preview=StaticBool()) {
   bUsePreviewValueAsDefault=true
   description="Since the volume texture cannot be blurred, you can store offset values in the R,G,B channels and average them to achieve a chromatic abberation style blur."
   inputName=Blur using offset RGB texture
   inputType=FunctionInput_StaticBool
   sortPriority=10
  }
  FunctionOutput(None=Clamp(None=StaticSwitch(True=@n26, False=@n27, Value=@n28) {desc=Blur Switch}) {desc=Final Shadow Mask}) {outputName=Shadow}
  FunctionOutput(None=@n6) {outputName=XY, sortPriority=1}
  FunctionOutput(None=@n5.W) {outputName=Z, sortPriority=2}
  FunctionOutput(None=@n15) {outputName=Internal Transmission Mask}
  Multiply(A=@n6) {constB=-1}
 ]}
