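# CheckerPattern is a library material function that creates a black-and-white checker pattern from input UVs and a Tiling input setting boxes per UV.
# UVs are scaled by half the tiling, then split into U and V.
# Frac, add half and Floor turn U, negated U and V into square waves.
# The V wave lerps between the two opposite U waves, forming the checker without branching.
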MaterialFunction /Engine/ArtTools/RenderToTexture/MaterialFunctions/CheckerPattern
graph {
 properties {bExposeToLibrary=true, description=Created a checker Pattern}
 nodes [
  n1 = TextureCoordinate() {uTiling=2, vTiling=2}
  n2 = FunctionInput() {
   bUsePreviewValueAsDefault=true
   description=How many checker boxes per UV
   inputName=Tiling
   inputType=FunctionInput_Vector2
   sortPriority=1
   previewValue=<1 1 0 1>
  }
  n3 = MaterialFunctionCall(Float2=Multiply(A=FunctionInput(Preview=@n1) {inputName=UVs, inputType=FunctionInput_Vector2}, B=Multiply(A=@n2) {constB=0.5}) {constB=0.5}) {materialFunction=&BreakOutFloat2Components}
  n4 = Add(A=Frac(None=Multiply(A=@n3.R) {constB=-1})) {constB=0.5}
  n5 = Add(A=Frac(None=@n3.R)) {constB=0.5}
  n6 = Add(A=Frac(None=@n3.G)) {constB=0.5}
  FunctionOutput(None=LinearInterpolate(A=Floor(None=@n4), B=Floor(None=@n5), Alpha=Floor(None=@n6) {desc=Vertical Blend})) {description=Creates a checker pattern from the desired UVs}
 ]
}
