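# FoliageZRotation is a library material function that returns a foliage object's rotation around Z as a 0-1 value, intended for foliage imposter materials.
# It transforms the constant X axis from local space, then feeds the result to VectorToRadialValue.
# The Vector Converted to Angle result becomes the single Z Rotation output, so the function needs no inputs.
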MaterialFunction /Engine/ArtTools/RenderToTexture/MaterialFunctions/FoliageZRotation
graph {
 properties {bExposeToLibrary=true, description=Returns foliage Z rotation. Useful for foliage Imposter materials.}
 nodes [
  n1 = Constant3Vector() {constant=<rgba 1 0 0 0>}
  FunctionOutput(None=MaterialFunctionCall(Vector or UVs=Transform(None=@n1) {transformSourceType=TRANSFORMSOURCE_Local}) {materialFunction=&VectorToRadialValue}.Vector Converted to Angle) {description=Returns the Z rotation, outputName="Z Rotation (0-1)"}
 ]
}
