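# Material function for impostor sprites that show a view-dependent atlas frame on a camera-facing quad.
# Camera yaw (atan2, offset by Rotation) and elevation pick the FramesX/FramesY cell, outputting UVs and a fractional blend phase.
# World position offset builds the sized billboard, twisted to hide angle steps; a Fixed Z switch instead uses upright sprites and flipbook frames.
# A costly extra output rotates normals into the view-aligned basis.
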MaterialFunction /Engine/ArtTools/RenderToTexture/MaterialFunctions/ImposterUVs
graph {
 properties {bExposeToLibrary=true, description=Imposter UV Material Function. Makes UVs and Sprite WorldPositionOffset}
 nodes [
  n1 = FunctionInput() {bUsePreviewValueAsDefault=true, inputName=Rotation, inputType=FunctionInput_Scalar, sortPriority=7}
  n2 = FunctionInput(Preview=ObjectPositionWS()) {bUsePreviewValueAsDefault=true, inputName=Position, sortPriority=4}
  n3 = Subtract(A=CameraPositionWS(), B=@n2)
  n4 = Multiply(A=@n3)
  MaterialExpressionComponentMask_4 = ComponentMask(None=@n4) {r=true}
  MaterialExpressionComponentMask_5 = ComponentMask(None=@n4) {g=true}
  n5 = Custom!(x=@MaterialExpressionComponentMask_4, y=@MaterialExpressionComponentMask_5) {
   code="return (atan2(y,x));"
   description=atan2
   outputType=CMOT_Float1
   inputs [
    {inputName=x, input {expression=&self:MaterialExpressionComponentMask_4}}
    {inputName=y, input {expression=&self:MaterialExpressionComponentMask_5}}
   ]
  }
  n6 = Multiply(A=Divide(A=ComponentMask(None=@n5) {r=true}, B=Constant() {r=6.2832})) {constB=-1}
  n7 = Frac(None=Add(A=Add(A=@n1, B=@n6)) {constB=0.25})
  n8 = FunctionInput() {bUsePreviewValueAsDefault=true, inputName=FramesX, inputType=FunctionInput_Scalar, previewValue=<24 0 0 1>}
  n9 = Multiply(A=@n8)
  n10 = Subtract(A=Multiply(A=@n7, B=@n9) {constB=6}, B=ScalarParameter() {defaultValue=0.375, parameterName=ImposterRoundingOffsetX}) {constB=0.5}
  n11 = Ceil(None=@n10)
  n12 = DotProduct(A=Normalize(VectorInput=@n3), B=Constant3Vector() {constant=<rgba 0 0 -1 0>})
  n13 = FunctionInput() {bUsePreviewValueAsDefault=true, inputName=FramesY, inputType=FunctionInput_Scalar, sortPriority=1, previewValue=<24 0 0 1>}
  n14 = Multiply(A=@n13)
  n15 = Subtract(
   A=Multiply(A=ConstantBiasScale(None=@n12) {desc=-1 to 1}, B=@n14) {constB=6}
   B=ScalarParameter() {defaultValue=1, parameterName=ImposterRoundingOffsetY}
  ) {constB=0.5}
  n16 = Subtract(A=@n14)
  n17 = Min(A=Ceil(None=@n15), B=@n16) {desc=Prevent looping on Y}
  n18 = Add(A=Divide(A=TextureCoordinate(), B=AppendVector(A=@n8, B=@n13)) {constB=48}) {constB=0}
  Add(A=AppendVector(A=Divide(A=@n11, B=@n9), B=Divide(A=@n17, B=@n14)), B=@n18) {desc=Add Frame Coords}
  Constant2Vector() {r=0.5}
  n19 = Multiply(A=@n8, B=@n13)
  n20 = Multiply(A=@n7, B=@n19)
  n21 = Frac(None=@n10) {desc=Blend Phase}
  n22 = Frac(None=@n15)
  n23 = FunctionInput(Preview=StaticBool()) {bUsePreviewValueAsDefault=true, inputName=Fixed Z, inputType=FunctionInput_StaticBool, sortPriority=2}
  FunctionOutput(None=StaticSwitch(True=Frac(None=@n20), False=AppendVector(A=@n21, B=@n22), Value=@n23)) {outputName=BlendPhase xy, sortPriority=1}
  n24 = Floor(None=@n15)
  n25 = AppendVector(A=Divide(A=Floor(None=@n10), B=@n9), B=Divide(A=Min(A=@n24, B=@n16) {desc=Prevent looping on Y}, B=@n14))
  FunctionOutput(None=StaticSwitch(
    True=MaterialFunctionCall("Animation  Phase (0-1)"=@n7, Number of Rows=@n8, Number of Columns=@n13) {materialFunction=&FlipBook}.UVs
    False=Add(A=@n25, B=@n18) {constB=0, desc=Add Frame Coords}
    Value=@n23
   )) {outputName=UVs}
  n26 = MaterialFunctionCall!(Vector1=@n3, Vector2=Constant3Vector() {constant=<rgba 0 0 1 0>}) {desc="No Rotation, for Vertex Shader", materialFunction=/Engine/Functions/Engine_MaterialFunctions02/Math/CreateThirdOrthogonalVector}
  n27 = StaticSwitch(True=Transform(None=Constant3Vector() {constant=<rgba 1 0 0 0>}) {transformSourceType=TRANSFORMSOURCE_View}, False=@n26.Vector3, Value=@n23)
  n28 = FunctionInput() {bUsePreviewValueAsDefault=true, inputName=SpriteWidth, inputType=FunctionInput_Scalar, sortPriority=3, previewValue=<4 0 0 1>}
  n29 = Multiply(A=@n28) {constB=-1}
  n30 = TextureCoordinate()
  n31 = MaterialFunctionCall(Float2=OneMinus(None=@n30)) {materialFunction=&BreakOutFloat2Components}
  n32 = Multiply(A=@n21, B=@n12)
  n33 = Multiply(
   A=Divide(B=@n8)
   B=MaterialFunctionCall(Multiplier=Constant() {r=-2}) {materialFunction=&Pi}.Result
  )
  n34 = Multiply(A=@n33) {constB=-0.5}
  n35 = Multiply(A=Add(A=Multiply(A=@n32, B=@n33), B=Multiply(A=@n12, B=@n34)), B=Multiply(A=ScalarParameter() {defaultValue=1, parameterName=ZRot}))
  n36 = Rotator(Coordinate=@n30, Time=@n35) {speed=1}
  n37 = StaticSwitch(
   True=AppendVector(A=ConstantBiasScale(None=@n31.R) {bias=-0.5, scale=2}, B=Multiply(A=@n31.G) {constB=2})
   False=ConstantBiasScale(None=OneMinus(None=@n36)) {bias=-0.5, scale=2}
   Value=@n23
  )
  n38 = StaticSwitch(True=Constant3Vector() {constant=<rgba 0 0 1 0>}, False=@n26.Vector2, Value=@n23)
  n39 = FunctionInput() {bUsePreviewValueAsDefault=true, inputName=SpriteHeight, inputType=FunctionInput_Scalar, sortPriority=4, previewValue=<4 0 0 1>}
  n40 = Add(
   A=Multiply(A=Multiply(A=@n27, B=@n29) {constB=-1}, B=ComponentMask(None=@n37) {r=true})
   B=Multiply(A=Multiply(A=@n38, B=@n39) {constB=-1}, B=ComponentMask(None=@n37) {g=true})
  )
  n41 = WorldPosition()
  FunctionOutput(None=Add(A=Subtract(A=@n40, B=@n41.XYZ), B=@n2) {constB=0}) {outputName=WorldPositionOffset, sortPriority=3}
  n42 = ScalarParameter() {parameterName=NORMAL ROTATE AMT}
  n43 = FunctionInput() {bUsePreviewValueAsDefault=true, inputName=Normals, sortPriority=9}
  n44 = RotateAboutAxis(
   NormalizedRotationAxis=Constant3Vector() {constant=<rgba 0 0 1 0>}
   RotationAngle=Add(A=Multiply(A=@n7, B=@n42), B=Divide(A=@n35, B=@n8))
   PivotPoint=Constant()
   Position=@n43
  )
  n45 = VectorParameter() {parameterName=RIGHT, defaultValue=<rgba 1 0 0 0>}
  n46 = VectorParameter() {parameterName=UP, defaultValue=<rgba 0 -1 0 0>}
  n47 = VectorParameter() {parameterName=FWD, defaultValue=<rgba 0 0 -1 0>}
  FunctionOutput(None=MaterialFunctionCall(
    VectorToTransform=Add(A=@n44, B=@n43)
    BasisX=Transform(None=@n45.RGB) {desc=RIGHT, transformSourceType=TRANSFORMSOURCE_View}
    BasisY=Transform(None=@n46.RGB) {desc=UP, transformSourceType=TRANSFORMSOURCE_View}
    BasisZ=Transform(None=@n47.RGB) {desc="FWD\r\n", transformSourceType=TRANSFORMSOURCE_View}
   ) {materialFunction=&Transform3x3Matrix}.Transformed Vector) {description="Transforms normals according to imposter rotation. Expensive!", outputName=TransformedNormals, sortPriority=2}
  MaterialFunctionCall(Vector or UVs=@n4) {materialFunction=&VectorToRadialValue}
  MaterialFunctionCall("Animation  Phase (0-1)"=Add(A=@n7, B=Divide(A=@n7, B=@n19)) {desc=Phase offset 1 frame}, Number of Rows=@n8, Number of Columns=@n13) {desc=Phase2 - For Blending between frames, materialFunction=&FlipBook}
  n48 = FunctionInput(Preview=MaterialFunctionCall() {materialFunction=&/FoliageZRotation}."Z Rotation (0-1)") {
   bUsePreviewValueAsDefault=true
   description="Since the rotation transform only works with CustomUVs on Foliage Meshes, and the main function should be using CustomUVs, this provides a way to feed in the rotation computer via CustomUVs for the normal map transformation."
   inputName=Rotation For Normals
   inputType=FunctionInput_Scalar
   sortPriority=8
  }
  Multiply(A=@n48) {constB=-1}
  n49 = CameraVectorWS()
  n50 = CrossProduct(A=@n49, B=Constant3Vector() {constant=<rgba 0 0 1 0>})
  Normalize(VectorInput=@n50)
  Normalize(VectorInput=CrossProduct(A=@n50, B=@n49))
  WorldPosition() {worldPositionShaderOffset=WPT_ExcludeAllShaderOffsets}
 ]
}
