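# Material function that outputs a UVW vector unchanged when it lies in the 0-1 range and replaces it with red otherwise.
# Each component passes through an If centered on 0.5 with an equality threshold just over 0.5, a single-node range test returning 1 inside and 0 outside.
# Min nodes combine the three masks into the alpha of a lerp from red to the input.
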MaterialFunction /Engine/ArtTools/RenderToTexture/MaterialFunctions/MarkRedOutside0-1Range
graph {nodes [
  n1 = FunctionInput() {bUsePreviewValueAsDefault=true, inputName=UVW, previewValue=<0.5 0.5 0.5 1>}
  n2 = MaterialFunctionCall(Float3=@n1) {materialFunction=&BreakOutFloat3Components}
  n3 = Constant()
  n4 = Constant()
  n5 = Min(
   A=If(A=@n2.R, "A > B"=@n3, "A == B"=Constant() {r=1}, "A < B"=@n3) {constB=0.5, equalsThreshold=0.5001}
   B=If(A=@n2.G, "A > B"=@n4, "A == B"=Constant() {r=1}, "A < B"=@n4) {constB=0.5, equalsThreshold=0.5001}
  )
  n6 = Constant()
  n7 = If(A=@n2.B, "A > B"=@n6, "A == B"=Constant() {r=1}, "A < B"=@n6) {constB=0.5, equalsThreshold=0.5001}
  FunctionOutput(None=LinearInterpolate(A=Constant3Vector() {constant=<rgba 1 0 0 0>}, B=@n1, Alpha=Min(A=@n5, B=@n7)))
 ]}
